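; Specter "Updoot" weapon that lets a Specter pawn reward a locked-on player.
; Lock-on attaches a Niagara target marker whose cooldown display refreshes through a next-tick timer.
; Trigger presses check cooldown, blocking tags and target, then send the target to the server.
; The server either launches a homing projectile or, after delays, applies a gameplay effect and tosses a weighted-random DataTable pickup with analytics, cue and auto-ping.
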(event EventBeginPlay
  (bind _instigator (Variables|Actor|GetInstigator))
  (|CallParentFunction)
  (Utilities|IsValid _instigator
    (:Is%20Valid
      (bind _asspecter_pawn (Utilities|Casting|CastToSpecterPawn _instigator))
      (Variables|Default|SetPlayer _asspecter_pawn)
      (EventDispatchers|BindEventtoOnPressTrigger (EventDispatchers|CreateEvent :!event Custom|TriggerPressed)))))

(event EventEndPlay
  (params (EndPlayReason :type EEndPlayReason%20Enum :default "Destroyed"))
  (|CallParentFunction EndPlayReason))

(event Custom|TriggerPressed
  (params (Weapon :type Fort%20Weapon%20Object%20Reference))
  ; Trigger Pressed
  (CallFunction|OnTriggerPressed))

(event FortWeapon|EventOnLocalUnEquip
  (Development|PrintString "On Local Unequip:")
  (CallFunction|SetLockonTargetActor 0))

(event Custom|Server_UpdootAtAutoAimTarget
  (params (AutoTargetActor :type Actor%20Object%20Reference))
  (bind (_returnvalue_location _returnvalue_rotation) (Math|Transform|BreakTransform (Transformation|GetActorTransform (Game|GetInstigator))))
  ; Server RPC - Fire At AutoAim Target
  (Variables|Default|SetCached_UpdootedTarget AutoTargetActor)
  (if (ScalableFloat|GetValueasBool (Variables|Default|GetScalableFloat_SpawnProjectile))
    ; SpawnProjectile
    (bind _returnvalue (Projectile|SpawnProjectile :ProjectileClass "/SpecterGameplay/items/UpdootItem/B_Prj_Ranged_SpecterUpdoot.B_Prj_Ranged_SpecterUpdoot_C" :RequestedBy self :SpawnLocation (Utilities|Operators|Add _returnvalue_location (Math|Vector|MakeVector 0.0 0.0 (Variables|Default|GetProjectileZOffset))) :SpawnRotation (Math|Conversions|RotationFromXVector (Math|Vector|RotateVectorAroundAxis (Math|Vector|GetForwardVector _returnvalue_rotation) -45.0 (Math|Vector|GetRightVector _returnvalue_rotation))) :SpawnDirection (Math|Conversions|RotationFromXVector (Math|Vector|RotateVectorAroundAxis (Math|Vector|GetForwardVector _returnvalue_rotation) -45.0 (Math|Vector|GetRightVector _returnvalue_rotation))) :bUseDefaultPhysics false :InitialSpeed 1e+02 :GravityScale 0.0 :ChargePercent 0.0 :OptionalAssociatedItemDef "/SpecterGameplay/items/WID_Specter_Updoot.WID_Specter_Updoot"))
    (bind _asb_prj_ranged_specter_updoot (Utilities|Casting|CastToB_Prj_Ranged_SpecterUpdoot _returnvalue))
    (Variables|/SpecterGameplay/items/UpdootItem/B_Prj_Ranged_SpecterUpdoot.B_Prj_Ranged_SpecterUpdoot_C|SetHoming%20Target%20Repped (Variables|Default|GetCached_UpdootedTarget) _asb_prj_ranged_specter_updoot)
    (else
      ;; Delay and show updoot
      (Utilities|FlowControl|Delay (ScalableFloat|GetValueatLevel (Variables|Default|GetShowUpdootDelaySeconds)))
      (bind _asfort_player_pawn (Utilities|Casting|CastToFortPlayerPawn (Variables|Default|GetCached_UpdootedTarget)))
      ;;;[0 0 0] Apply effect for character and screen fx on hit player
      (CallFunction|ApplyUpdootGameplayEffect _asfort_player_pawn)
      ; Delay and spawn pickup
      (Utilities|FlowControl|Delay (ScalableFloat|GetValueatLevel (Variables|Default|GetSpawnPickupDelaySeconds)))
      (CallFunction|SpawnDootPickup (Variables|Default|GetCached_UpdootedTarget)))))

(fn ApplyUpdootGameplayEffect
  (params (FortPlayerPawn :type Fort%20Player%20Pawn%20Object%20Reference))
  (bind _output_get (LocalVariables|SetWeaponInstigator (Game|GetInstigator)))
  (Utilities|IsValid _output_get
    (:Is%20Valid
      (GameplayEffects|ApplyGameplayEffectSpecToTarget (Ability|GetAbilitySystemComponent _output_get) (GameplayEffects|MakeOutgoingSpec (Ability|GetAbilitySystemComponent _output_get) (Variables|Default|GetOnHitGameplayEffect) 0.0 (Class|AthenaFunctionLibrary|MakeSimpleFortEffectContext _output_get self)) (Ability|GetAbilitySystemComponent FortPlayerPawn)))))

(fn FrameUpdateTargetNiagaraSystemStatus
  (bind _targetedactorspriteniagarasystem (Variables|Default|GetTargetedActorSpriteNiagaraSystem))
  (bind (_timeremaining _cooldownduration) (Ability|Cooldowns|GetCooldownTimeRemainingandDuration (Game|GetInstigator) "(GameplayTags=((TagName=\"Specter.UpdootItem.Cooldown\")))"))
  (Niagara|SetNiagaraVariable(Bool) (Variables|Default|GetTargetedActorSpriteNiagaraSystem) "Cooldown Active" (> _timeremaining 0))
  (if (> _cooldownduration 0)
    (Niagara|SetNiagaraVariable(Float) _targetedactorspriteniagarasystem "Cooldown Alpha" (/ _timeremaining _cooldownduration))
    (else
      (Niagara|SetNiagaraVariable(Float) _targetedactorspriteniagarasystem "Cooldown Alpha" 0.0)))
  (bind _returnvalue (Utilities|Time|SetTimerforNextTickbyFunctionName 0 "FrameUpdateTargetNiagaraSystemStatus"))
  (Variables|Default|SetUpdateTargetSpriteStatusTimerHandle _returnvalue))

(fn GetLockOnTargetActor
  (return (Variables|Default|GetCurrenttLockOnTargetedActor)))

(fn IsFiringBlocked
  (bind _returnvalue (Ability|Cooldowns|GetCooldownTimeRemaining (Game|GetInstigator) "(GameplayTags=((TagName=\"Specter.UpdootItem.Cooldown\")))"))
  (return (Math|Boolean|ORBoolean (> _returnvalue 0) (GameplayTags|HasAnyMatchingGameplayTags (Variables|Default|GetPlayer) (Variables|Default|GetBlockingTags)) (not (Utilities|IsValid (Variables|Default|GetCurrenttLockOnTargetedActor) :!class /Script/Engine.KismetSystemLibrary)) (not (FortWeapon|CanFire true)))))

(fn OnTriggerPressed
  (CallFunction|TryToFire))

(fn SetLockOnTargetActor
  (params (InLocalCurrentlyTargetedActor :type Actor%20Object%20Reference))
  (bind _updatetargetspritestatustimerhandle (Variables|Default|GetUpdateTargetSpriteStatusTimerHandle))
  (bind _targetedactorspriteniagarasystem (Variables|Default|GetTargetedActorSpriteNiagaraSystem))
  (if (!= InLocalCurrentlyTargetedActor (Variables|Default|GetCurrenttLockOnTargetedActor))
    (bind _output_get (Variables|Default|SetCurrenttLockOnTargetedActor InLocalCurrentlyTargetedActor))
    (Utilities|FlowControl|Sequence
      (:then_0
        ; If we already have an active system stop it.
        (Utilities|IsValid _targetedactorspriteniagarasystem
          (:Is%20Valid
            (Components|Activation|Deactivate _targetedactorspriteniagarasystem)
            (Transformation|DetachFromComponent _targetedactorspriteniagarasystem "KeepRelative" "KeepRelative" "KeepRelative" true)
            (Effects|Components|ParticleSystem|ReleasetoPool _targetedactorspriteniagarasystem)
            (Variables|Default|SetTargetedActorSpriteNiagaraSystem 0)
            (if (Utilities|Time|IsTimerActivebyHandle _updatetargetspritestatustimerhandle)
              (Utilities|Time|ClearandInvalidateTimerbyHandle _updatetargetspritestatustimerhandle)))))
      (:then_1
        ; Start a new system if we have a valid actor.
        (Utilities|IsValid _output_get
          (:Is%20Valid
            (bind _asfort_player_pawn (Utilities|Casting|CastToFortPlayerPawn _output_get))
            (bind _returnvalue (Niagara|SpawnSystemAttached "/SpecterGameplay/Gameplay/Effects/Updoot/NS_Updoot_Target.NS_Updoot_Target" (Variables|/Script/Engine.Character|GetMesh _asfort_player_pawn) "spine_01" "15.000000,0.000000,0.000000" "0, 0, 0" "KeepRelativeOffset" true true "ManualRelease"))
            (Variables|Default|SetTargetedActorSpriteNiagaraSystem _returnvalue)
            (CallFunction|FrameUpdateTargetNiagaraSystemStatus)))))))

(fn SpawnDootPickup
  (params (InLockOnTargetActor :type Actor%20Object%20Reference))
  (bind _local_targetpawn (LocalVariables|GetLocal_TargetPawn))
  (bind _returnvalue (Fortnite|GetFortPlayerControllerfromActor _local_targetpawn))
  (bind _returnvalue_2 (Class|FortPlayerController|GetMarkerComponent _returnvalue))
  (bind _player (Variables|Default|GetPlayer))
  ;[1 1 1 0.5] SpawnDootPickup: First iterate to add up all weights
  (bind _outrownames (DataTable|GetDataTableRowNames (Variables|Default|GetDataTableOfPickups)))
  (if (> (Utilities|Array|Length _outrownames) 0)
    (LocalVariables|SetLocal_WeightTotal 0.0)
    (for _array_element _outrownames
      (bind _returnvalue_3 (Utilities|GetDataTableRow (Variables|Default|GetDataTableOfPickups) _array_element))
      (LocalVariables|SetLocal_WeightTotal (+ (LocalVariables|GetLocal_WeightTotal) (.Weight_11_187623D1452BDEE461E9CCB2D37C5329 (Utilities|Struct|BreakStructSpecterSpawnPickupsRow _returnvalue_3)))))
    ;[1 1 1 0.5] Randomly pick a value within our weight total
    (if (> (LocalVariables|GetLocal_WeightTotal) 0)
      (LocalVariables|SetLocal_TargetValue (Math|Random|RandomFloatinRange 0.0 (LocalVariables|GetLocal_WeightTotal)))
      ;[1 1 1 0.5] Iterate again, this time to find our target element.
      (LocalVariables|SetLocal_WeightTotal 0.0)
      (bind _array_element_2 (Utilities|Array|ForEachLoopwithBreak _outrownames)
        (:LoopBody
          (bind _returnvalue_4 (Utilities|GetDataTableRow (Variables|Default|GetDataTableOfPickups) _array_element_2))
          (LocalVariables|SetLocal_Weight (.Weight_11_187623D1452BDEE461E9CCB2D37C5329 (Utilities|Struct|BreakStructSpecterSpawnPickupsRow _returnvalue_4)))
          (LocalVariables|SetLocal_ItemSpawned (Utilities|String|ToString(Object) (Utilities|Struct|BreakStructSpecterSpawnPickupsRow _returnvalue_4)))
          (bind _output_get (LocalVariables|SetLocal_WeightTotal (+ (LocalVariables|GetLocal_WeightTotal) (LocalVariables|GetLocal_Weight))))
          (if (<= (LocalVariables|GetLocal_TargetValue) _output_get)
            (bind _asfort_player_pawn (Utilities|Casting|CastToFortPlayerPawn (Variables|GetInLockOnTargetActor)))
            (bind _output_get_2 (LocalVariables|SetLocal_TargetPawn _asfort_player_pawn))
            (bind _returnvalue_5 (Inventory|SpawnPickupInWorld (Utilities|Struct|BreakStructSpecterSpawnPickupsRow _returnvalue_4) 1 (+ (Transformation|GetWorldLocation (Variables|/Script/Engine.Character|GetCapsuleComponent _output_get_2)) (* (Transformation|GetForwardVector (Variables|/Script/Engine.Character|GetCapsuleComponent _output_get_2)) 5e+01)) "1.000000,0.000000,0.000000" 0 false false)) ; Pistol C
            (Pickup|TossPickup _returnvalue_5 (+ (Transformation|GetWorldLocation (Variables|/Script/Engine.Character|GetCapsuleComponent _output_get_2)) (* (Transformation|GetForwardVector (Variables|/Script/Engine.Character|GetCapsuleComponent _output_get_2)) 7.5e+02)) (LocalVariables|GetLocal_TargetPawn) 0 true true "Tossed" "Unset")
            ;; log the doot for analytics
            (Utilities|IsValid _player
              (:Is%20Valid
                (Class|SpecterPawn|LogDootAnalytics _player (LocalVariables|GetLocal_TargetPawn) (+ (Transformation|GetWorldLocation (Variables|/Script/Engine.Character|GetCapsuleComponent _output_get_2)) (* (Transformation|GetForwardVector (Variables|/Script/Engine.Character|GetCapsuleComponent _output_get_2)) 7.5e+02)) (LocalVariables|GetLocal_ItemSpawned))
                ;[1 0 0.5463] Pickup spawned gameplaycue
                (Utilities|IsValid _returnvalue_5
                  (:Is%20Valid
                    (LocalVariables|SetLocal_SpawnedPickup _returnvalue_5)
                    (GameplayCue|ExecuteGameplayCueOnActor(Burst) (LocalVariables|GetLocal_TargetPawn) (Variables|Default|GetSpawnedPickupGameplayCue) (Ability|GameplayCue|MakeGameplayCueParameters 0.0))
                    ; Auto-ping spawned pickup for the intended owner
                    (Utilities|IsValid _returnvalue
                      (:Is%20Valid
                        (Utilities|IsValid _returnvalue_2
                          (:Is%20Valid
                            (MarkerComponent|MarkActoronServer _returnvalue_2 (LocalVariables|GetLocal_SpawnedPickup) false false))
                          (:Is%20Not%20Valid
                            _)))
                      (:Is%20Not%20Valid
                        _)))
                  (:Is%20Not%20Valid
                    _)))
              (:Is%20Not%20Valid
                _)))))
      (else
        (Fortnite|FortShippingLog self "GA_SpecterUpdootPrimary: Failed because data table had invalid weights." true)))
    (else
      (Fortnite|FortShippingLog self "GA_SpecterUpdootPrimary: Failed to spawn actor for updoot because table is empty." true))))

(fn TrytoFire
  (if (Default|IsFiringBlocked)
    ; Unable to fire
    (FailedAction|UnabletoPerformAction (Variables|Default|GetPlayer) true (GameplayTags|MakeLiteralGameplayTagContainer "(GameplayTags=((TagName=\"Athena.Ability.Failed.Failed.GenericFail\")))")) ; show reason for failing on screen if one is given
    (else
      ; Fire
      (CallFunction|ServerUpdootatAutoAimTarget (Variables|Default|GetCurrenttLockOnTargetedActor))
      (CallFunction|TrytoCustomFire false))))

(fn ConstructionScript
  (|CallParentFunction))
